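# One-shot mono MetaSound source for melee weapon swing sounds, with melee attenuation, sound class and concurrency settings.
# On play, it shuffles a swing wave array, empty by default, and feeds the pick to a wave player.
# Base-plus-or-minus-spread inputs randomize pitch shift and decibel volume per play.
# Volume converts to linear gain, scales by MasterVolume and multiplies the audio; playback end fires OnFinished.
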MetaSoundSource /FortMetasound/Weapons/MeleeWeapons/MSS_Weapon_Melee_Swing
inputs [
 MasterVolume {type=Float}
 Random.PitchShift.Base {type=Float}
 Random.PitchShift.Spread {type=Float}
 Random.Volume.Base {type=Float}
 Random.Volume.Spread {type=Float}
 Sounds.Swings {type=WaveAsset:Array}
 UE.Source.OnPlay {type=Trigger}
]
interfaces [UE.OutputFormat.Mono:1.0, UE.Source.OneShot:1.0, UE.Source:1.1]
outputs [
 "UE.OutputFormat.Mono.Audio:0" {type=Audio}
 UE.Source.OneShot.OnFinished {type=Trigger}
]
pages [Default {
  input_defaults {
   MasterVolume=1
   Random.PitchShift.Base=0
   Random.PitchShift.Spread=0.5
   Random.Volume.Base=0
   Random.Volume.Spread=0
   UE.Source.OnPlay=null
   Sounds.Swings []
  }
  nodes [
   n1 = InitVariable.Float()
   DecibelstoLinearGain_0 = UE.Decibels to Linear Gain:Float(Decibels=UE.RandomFloat(
     Next=^UE.Source.OnPlay
     Max=UE.Add:Float(PrimaryOperand=^Random.Volume.Base, AdditionalOperands=^Random.Volume.Spread).Out
     Min=UE.Subtract:Float(PrimaryOperand=^Random.Volume.Base, AdditionalOperands=^Random.Volume.Spread).Out
    ).Value)
   n2 = UE.Multiply:Float(AdditionalOperands=^MasterVolume, PrimaryOperand=@DecibelstoLinearGain_0.Linear Gain)
   n3 = Array.Shuffle:WaveAsset:Array(In Array=^Sounds.Swings, Next=^UE.Source.OnPlay) {also {defaults {Enable Shared State="<MetaSoundBuilderResult.SUCCEEDED: 0>"}}}
   n4 = InitVariable.Float()
   n5 = UE.RandomFloat(
    Next=^UE.Source.OnPlay
    Max=UE.Add:Float(PrimaryOperand=^Random.PitchShift.Base, AdditionalOperands=^Random.PitchShift.Spread).Out
    Min=UE.Subtract:Float(PrimaryOperand=^Random.PitchShift.Base, AdditionalOperands=^Random.PitchShift.Spread).Out
   )
   WavePlayer_0 = UE.Wave Player:Mono(
    Play=@n3.On Next
    Wave Asset=@n3.Value
    Pitch Shift=VariableAccessor.Float(Variable=VariableMutator.Float(Variable=@n4.Variable, Value=@n5.Value).Variable).Value
   ) {also {version [1,1]}}
   Multiply_0 = UE.Multiply:Audio by Float(
    AdditionalOperands=VariableAccessor.Float(Variable=VariableMutator.Float(Variable=@n1.Variable, Value=@n2.Out).Variable).Value
    PrimaryOperand=@WavePlayer_0.Out Mono
   )
  ]
  links [{dst="@UE.OutputFormat.Mono.Audio:0", src=Multiply_0.Out}, {dst="@UE.Source.OneShot.OnFinished", src=WavePlayer_0.On Finished}]
 }]
properties {
 assetImportData=&self:AssetImportData
 attenuationSettings=&/ATT_Weapon_Melee_Attack
 audioPropertiesSheet=&/APS_Weapon_Melee
 maxDistance=12500
 sampleRate=48000
 soundClassObject=&Fort_WeaponMelee_PSM
 concurrencySet [&Player_Melee_Weapon_Conc]
}
