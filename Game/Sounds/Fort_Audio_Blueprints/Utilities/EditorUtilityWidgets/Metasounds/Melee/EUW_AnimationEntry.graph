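# Editor utility widget for one animation row in a melee audio tool, with a button to open the asset and another to remove the row.
# AddNotify takes a sound tag and inspects existing notifies, offering to replace PlaySound notifies and checking SoundLibrary notifies for that tag.
# It can add a SoundLibrary notify on a MeleeAudio track, building its RightHand-attached context and assigning it through SetEditorProperty, then saves the animation.
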EditorUtilityWidgetBlueprint /Game/Sounds/Fort_Audio_Blueprints/Utilities/EditorUtilityWidgets/Metasounds/Melee/EUW_AnimationEntry
graphs [EventGraph {nodes [
   "On Clicked (AssetButton)"() {is=title K2Node_ComponentBoundEvent}
   > n1 = Get!(execute:Exec) -> (then:Exec, else:Exec, Animation:Anim Sequence Base Object Reference)
   > n2 = LoadAsset(AssetPath=Get Object Path String(Object=@n1))
   > OpenEditorForAssets!(
    execute
    self:Asset Editor Subsystem Object Reference=Get AssetEditorSubsystem()
    Assets:Array of Object References=Make Array("[0]":Object Reference=@n2) -> (Array:Array of Object References)
    OpenedMethod:EAssetTypeActivationOpenedMethod Enum=Edit
   ) -> (then, ReturnValue:bool)
   "On Clicked (EditorUtilityButton_96)"() {is=title K2Node_ComponentBoundEvent}
   n3 = Self-Reference()
   > RemoveChild(self=GetParent!(self:Widget Object Reference=@n3) -> (ReturnValue:Panel Widget Object Reference), Content=@n3)
   n4 = SetAnimation!() -> (OutputDelegate, then, Animation:Anim Sequence Base Object Reference) {is=title K2Node_CustomEvent}
   > Is Valid(InputObject=@n4.Animation)
   > n5 = Set Animation(Animation:Anim Sequence Base Object Reference=@n4.Animation) -> (Output_Get:Anim Sequence Base Object Reference)
   > Set Text(Text="To Text (String)"(InString=GetDisplayName(Object=@n5)), self=Get AssetName() -> (AssetName:Text Object Reference))
   n6 = AddNotify!() -> (OutputDelegate, then, SoundLibraryEvent:Sound Library Tag Structure) {is=title K2Node_CustomEvent}
   > Set SoundTag(SoundTag:Gameplay Tag Structure=@n6.SoundLibraryEvent) -> (Output_Get:Gameplay Tag Structure)
   > n7 = Get()
   n8 = Get Animation() -> (Animation:Anim Sequence Base Object Reference)
   > AddAnimationNotifyTrack!(
    execute
    AnimationSequenceBase:Anim Sequence Base Object Reference=@n8
    NotifyTrackName:Name=MeleeAudio
    TrackColor:Linear Color Structure=(R=0.0389,G=0,B=1,A=1)
   ) -> (then)
   > n9 = AddAnimationNotifyEvent!(
    execute
    AnimationSequenceBase:Anim Sequence Base Object Reference=@n8
    NotifyTrackName:Name=MeleeAudio
    StartTime:float=0
    NotifyClass:Anim Notify Class Reference=/Script/SoundLibrary.AnimNotify_SoundLibrary
   ) -> (then, ReturnValue:Anim Notify Object Reference)
   > Set context!(execute, self, Notify:Anim Notify Object Reference=@n9) -> (then) {is=title K2Node_CallFunction}
   > SaveLoadedAsset(AssetToSave=Get Animation(), bOnlyIfIsDirty=true)
   n10 = NotifyReplaced!() -> (OutputDelegate, then, OldNotify:Anim Notify Object Reference, NewNotify:Anim Notify Object Reference) {is=title K2Node_CustomEvent}
   > Set context(Notify=@n10.NewNotify)
   n11 = Set context!() -> (OutputDelegate, then, Notify:Anim Notify Object Reference) {is=title K2Node_CustomEvent}
   n12 = Get SoundTag() -> (SoundTag:Gameplay Tag Structure)
   > Branch(Condition=IsGameplayTagValid!(GameplayTag:Gameplay Tag Structure=@n12) -> (ReturnValue:bool))
   > n13 = Cast To AnimNotify_SoundLibrary!(execute, Object=@n11.Notify) -> (then, CastFailed, AsAnim Notify Sound Library:Anim Notify Sound Library Object Reference) {is=title K2Node_DynamicCast}
   > n14 = Create Sound Library Context!(
    execute:Exec
    WorldContextObject:Object Reference
    Class:Sound Library Context Class Reference=/Script/SoundLibrary.SoundLibraryAnimContext
    AnimSettings:Animation Notify Settings Structure=Make Animation Notify Settings!(
     AnimRateThreshold:Float Interval Structure=(Min=0,Max=0)
     EventName:Gameplay Tag Structure=@n12
     bOnThisActor:bool=False
     bAttachToActor:bool=False
     bStopWhenAttachedDestroyed:bool=False
     Offset:Vector="0,0,0"
     AttachName:Name=RightHand
     FadeTime:float=0.5
     NotifyTriggerChance:float=1
     VolumeMultiplier:float=1
    ) -> (SoundLibraryAnimContextSettings:Animation Notify Settings Structure) {is=title K2Node_MakeStruct}
    Animation:Anim Sequence Base Weak object=Get Animation()
    Actor:Actor Weak object
    BaseEventName:Gameplay Tag Structure=(TagName="")
    bForwardToActorOwner:bool=True
    bForwardToAttachedActors:bool=True
    bAutoResetContext:bool=True
   ) -> (then:Exec, ReturnValue:Animation Notify Context Object Reference) {is=title K2Node_CreateSoundLibraryContext}
   > SetEditorProperty(Object=@n13, PropertyName=SoundContext, PropertyValue:Animation Notify Context Object Reference=@n14)
   n15 = EditPlaySound!() -> (OutputDelegate, then, AsPlaySound:Play Sound Object Reference) {is=title K2Node_CustomEvent}
   n16 = Append!(
    A="The animation "
    B=GetDisplayName(Object=Get Animation())
    C:String=" has an existing play sound notify with the sound "
    D:String=GetDisplayName(Object=Get Sound!(self:Play Sound Object Reference=@n15.AsPlaySound) -> (Sound:Sound Base Object Reference))
    E:String=", would you like to replace it with a sound library notify?"
   ) -> (ReturnValue)
   > n17 = Show Message Dialog(
    Title=MakeLiteralText(Value="LOCTEXT(\"Existing Sound found\")")
    Message=MakeLiteralText(Value="To Text (String)"(InString=@n16))
    MessageType=YesNo
    DefaultValue=No
    MessageCategory=Warning
   )
   > n18 = Switch on EAppReturnType(Selection=@n17)
   ReplaceAnimNotifies!(
    execute=@n18.Yes
    AnimationSequenceBase:Anim Sequence Base Object Reference=Get Animation()
    OldNotifyClass:Anim Notify Class Reference=/Script/Engine.AnimNotify_PlaySound
    NewNotifyClass:Anim Notify Class Reference=&AnimNotify_SoundLibrary
    OnNotifyReplaced:Delegate=@n10
   ) -> (then)
   > Set ReplacedNotify(ReplacedNotify:bool=true) -> (Output_Get:bool)
   n19 = EditSoundLibrary!() -> (OutputDelegate, then, AsAnimNotifySoundLibrary:Anim Notify Sound Library Object Reference) {is=title K2Node_CustomEvent}
   > n20 = GetEditorProperty!(execute:Exec, Object:Object Reference=@n19.AsAnimNotifySoundLibrary, PropertyName:Name=SoundContext) -> (then:Exec, PropertyValue:Animation Notify Context Object Reference, ReturnValue:bool)
   > n21 = Set Property Value(Property Value:Animation Notify Context Object Reference=@n20) -> (Output_Get:Animation Notify Context Object Reference)
   n22 = Get AnimSettings!(self:Animation Notify Context Object Reference=@n21) -> (AnimSettings:Animation Notify Settings Structure)
   > Set EventExists(EventExists:bool="Equal (GameplayTag)"(
     A=Break Animation Notify Settings!(SoundLibraryAnimContextSettings:"Animation Notify Settings Structure (by ref)"=@n22) -> (
      AnimRateThreshold:Float Interval Structure
      EventName:Gameplay Tag Structure
      bOnThisActor:bool=~
      bAttachToActor:bool=~
      bStopWhenAttachedDestroyed:bool=~
      Offset:Vector=~
      AttachName:Name
      FadeTime:float=~
      NotifyTriggerChance:float=~
      VolumeMultiplier:float=~
     ) {is=title K2Node_BreakStruct}.EventName
     B=Get SoundTag()
    )) -> (Output_Get:bool)
   n23 = For Each Loop(Array:Array of Anim Notify Event Structures=GetAnimationNotifyEvents!(AnimationSequenceBase:Anim Sequence Base Object Reference=@n7) -> (NotifyEvents:Array of Anim Notify Event Structures)) -> (Array Element:Anim Notify Event Structure)
   n24 = Get Class(Object=Break Anim Notify Event!(AnimNotifyEvent:"Anim Notify Event Structure (by ref)"=@n23) -> (Notify:Anim Notify Object Reference, NotifyStateClass:Anim Notify State Object Reference) {is=title K2Node_BreakStruct}) -> (ReturnValue:Anim Notify Class Reference)
   Branch(Condition="Equal (Class)"(A=@n24, B=Get As Play Sound() -> (As Play Sound:Play Sound Class Reference)))
   n25 = Reroute Node(InputPin:Object Reference) -> (OutputPin:Object Reference)
   > n26 = Cast To AnimNotify_PlaySound!(execute, Object=@n25) -> (then, CastFailed, AsPlay Sound:Play Sound Object Reference) {is=title K2Node_DynamicCast}
   > EditPlaySound!(execute, self, AsPlaySound:Play Sound Object Reference=@n26) -> (then) {is=title K2Node_CallFunction}
   n27 = Get!(execute) -> (then, else, ReplacedNotify:bool)
   Get!(execute=@n27.else) -> (then, else, EventExists:bool)
   n28 = Cast To AnimNotify_SoundLibrary(execute=@n30.then, Object=@n25)
   > EditSoundLibrary!(execute, self, AsAnimNotifySoundLibrary:Anim Notify Sound Library Object Reference=@n28) -> (then) {is=title K2Node_CallFunction}
   Event Construct()
   n29 = Sequence(execute=@n23.LoopBody)
   n30 = Branch(execute=@n29.then_1, Condition="Equal (Class)"(A=@n24, B=&AnimNotify_SoundLibrary))
  ]}]
variables [
 Animation {type=AnimSequenceBase, default=null}
 As Anim Notify Sound Library {type=AnimNotify_SoundLibrary, default=null}
 As Play Sound {type=Class, default=null}
 AudioNotifyClasses {type="array<Class>", default [/Script/SoundLibrary.AnimNotifyState_SoundLibrary, null]}
 Context {type=Class, default=&SoundLibraryAnimContext}
 EventExists {type=boolean}
 NewVar {type=SoundLibraryAnimContext, default=null}
 Property Value {type=SoundLibraryAnimContext, default=null}
 ReplacedNotify {type=boolean}
 SoundTag {type=GameplayTag, default {tagName=SoundLibrary.Ability}}
]
widgets {
 info {rootWidgetClass=&StackBox, widgetCount=7}
 widgets [
  StackBox_0 {widgetClassPath=&StackBox, widget=StackBox_0}
  HorizontalBox_0 {
   widgetClassPath=&HorizontalBox
   parent=StackBox_0
   widget=HorizontalBox_0
   slot {class=StackBoxSlot, properties {size {sizeRule=Fill, value=1}}}
  }
  AssetButton {
   widgetClassPath=&EditorUtilityButton
   parent=HorizontalBox_0
   widget=AssetButton
   slot {class=HorizontalBoxSlot, properties {size {sizeRule=Fill, value=1}}}
  }
  AssetName {
   widgetClassPath=&TextBlock
   bIsVariable=true
   parent=AssetButton
   widget=AssetName
   properties {
    justification=Center
    text=Name
    font {
     bForceMonospaced=false
     bMaterialIsStencil=false
     fontMaterial=null
     fontObject=&Burbank
     letterSpacing=0
     monospacedWidth=1
     size=15
     skewAmount=0
     typefaceFontName=Burbank-Black
     outlineSettings {
      bApplyOutlineToDropShadows=false
      bMiteredCorners=false
      bSeparateFillAlpha=false
      outlineMaterial=null
      outlineSize=0
      outlineColor=<rgba 0 0 0 1>
     }
    }
   }
   slot {class=ButtonSlot, properties {padding=<ltrb 0 0 0 0>}}
  }
  Spacer_282 {widgetClassPath=&Spacer, parent=HorizontalBox_0, widget=Spacer_282, properties {size=<21.5071 1>}, slot {class=HorizontalBoxSlot}}
  EditorUtilityButton_96 {widgetClassPath=&EditorUtilityButton, parent=HorizontalBox_0, widget=EditorUtilityButton_96, slot {class=HorizontalBoxSlot}}
  Image_48 {
   widgetClassPath=&Image
   bIsVariable=true
   parent=EditorUtilityButton_96
   widget=Image_48
   properties {brush {
     bIsDynamicallyLoaded=false
     drawAs=Image
     imageType=FullColor
     mirroring=NoMirror
     resourceName=None
     resourceObject=/Game/UI/Foundation/Textures/Buttons/T-SlotIcon-Remove-64
     tiling=NoTile
     imageSize=<25 25>
     margin=<ltrb 0 0 0 0>
     outlineSettings {
      bUseBrushTransparency=false
      roundingType=HalfHeightRadius
      width=0
      color {colorUseRule=UseColor_Specified, specifiedColor=<rgba 0 0 0 0>}
      cornerRadii=<0 0 0 0>
     }
     tintColor {colorUseRule=UseColor_Specified, specifiedColor=<rgba 1 1 1 1>}
     uVRegion {bIsValid=false, max=<0 0>, min=<0 0>}
    }}
   slot {class=ButtonSlot}
  }
 ]
}
