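# Editor utility widget for melee audio tooling that shows a scrollable "Animation List" of animation names.
# BuildWidgets appends incoming animations to the stored array, soft-resets the rows, then calls SetUpWidgetActions per animation to add a selection widget labeled with its name.
# That function can fill each row with AnimationEntry widgets, but BuildWidgets leaves its Animations input unconnected.
# PreConstruct hard-resets, clearing both animations and rows.
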EditorUtilityWidgetBlueprint /Game/Sounds/Fort_Audio_Blueprints/Utilities/EditorUtilityWidgets/Metasounds/Melee/EUW_MeleeAudio_AnimationName_ListView
graphs [
 EventGraph {nodes [
   n1 = Reset Widgets!() -> (OutputDelegate, then, HardReset:bool=~) {is=title K2Node_CustomEvent}
   > n2 = Branch(Condition=@n1.HardReset)
   > n3 = Clear(TargetArray:Array of Anim Sequence Base Object References=Get Animations() -> (Animations:Array of Anim Sequence Base Object References))
   n4 = For Each Loop(
    Exec=[@n3.then, @n2.else]
    Array:Array of EUW Melee Animation Selection Object References=Get AnimationNameWidgets() -> (AnimationNameWidgets:Array of EUW Melee Animation Selection Object References)
   ) -> (Array Element:EUW Melee Animation Selection Object Reference)
   Clear(execute=@n4.Completed, TargetArray:Array of EUW Melee Animation Selection Object References=Get AnimationNameWidgets())
   RemoveChild(execute=@n4.LoopBody, self=Get MainBox() -> (MainBox:Editor Utility Scroll Box Object Reference), Content=@n4)
   n5 = BuildWidgets!() -> (OutputDelegate, then, Animations:Array of Anim Sequence Base Object References) {is=title K2Node_CustomEvent}
   > Append Array!(
    execute
    TargetArray:Array of Anim Sequence Base Object References=Get Animations()
    SourceArray:Array of Anim Sequence Base Object References=@n5.Animations
   ) -> (then)
   > Reset Widgets!(execute, self, HardReset:bool=false) -> (then) {is=title K2Node_CallFunction}
   > n6 = For Each Loop(Array:Array of Anim Sequence Base Object References=Get Animations()) -> (Array Element:Anim Sequence Base Object Reference)
   > SetUpWidgetActions!(execute, self, ActionName:String=GetDisplayName(Object=@n6), Animations:Array of Anim Sequence Base Object References) -> (then)
   Event PreConstruct()
   > Reset Widgets(HardReset=true)
   > Clear(TargetArray=Get AnimationNameWidgets())
  ]}
 SetUpWidgetActions {nodes [
   SetUpWidgetActions!() -> (then, ActionName:String, Animations:Array of Anim Sequence Base Object References) {is=title K2Node_FunctionEntry}
   > n1 = Construct Object from Class(
    Class=&/EUW_Melee_AnimationSelection.EUW_Melee_AnimationSelection_C
    self=Get MainBox()
   ) -> (ReturnValue:EUW Melee Animation Selection Object Reference)
   > Add Unique(
    TargetArray:Array of EUW Melee Animation Selection Object References=Get AnimationNameWidgets()
    NewItem:"EUW Melee Animation Selection Object Reference (by ref)"=@n1
   )
   > AddChild(self=Get MainBox(), Content=@n1)
   > "SetText (Text)"(
    self=Get PresetText!(self:EUW Melee Animation Selection Object Reference=@n1) -> (PresetText:Text Object Reference)
    InText="To Text (String)"(InString=Get ActionName() -> (ActionName:String))
   )
   n2 = Get ScrollBox!(self:EUW Melee Animation Selection Object Reference=@n1) -> (ScrollBox:Editor Utility Scroll Box Object Reference)
   > SetAlwaysShowScrollbar!(execute, self:Scroll Box Object Reference=@n2, NewAlwaysShowScrollbar:bool=true) -> (then)
   n3 = Get Animations()
   > Branch(Condition=Is Not Empty!(TargetArray:Array of Anim Sequence Base Object References=@n3) -> (ReturnValue:bool))
   > n4 = For Each Loop(Array=@n3)
   > n5 = Construct Object from Class(Class=&/EUW_AnimationEntry.EUW_AnimationEntry_C) -> (ReturnValue:EUW Animation Entry Object Reference)
   > AddChild(self=@n2, Content=@n5)
   > SetAnimation!(execute, self:EUW Animation Entry Object Reference=@n5, Animation:Anim Sequence Base Object Reference=@n4) -> (then)
  ]}
]
variables [
 AnimationNameWidgets {type="array<EUW_Melee_AnimationSelection_C>", default []}
 Animations {type="array<AnimSequenceBase>", default []}
]
widgets {
 info {rootWidgetClass=&CanvasPanel, widgetCount=7}
 widgets [
  CanvasPanel_0 {widgetClassPath=&CanvasPanel, widget=CanvasPanel_0}
  VerticalBox_59 {
   widgetClassPath=&VerticalBox
   parent=CanvasPanel_0
   widget=VerticalBox_59
   slot {class=CanvasPanelSlot, properties {layoutData {alignment=<0 0>, anchors {maximum=<0 0>, minimum=<0 0>}, offsets=<ltrb 0 0 600 473.8739>}}}
  }
  Border_193 {
   widgetClassPath=&Border
   parent=VerticalBox_59
   widget=Border_193
   properties {
    horizontalAlignment=HAlign_Center
    background {
     bIsDynamicallyLoaded=false
     drawAs=RoundedBox
     imageType=NoImage
     mirroring=NoMirror
     resourceName=None
     resourceObject=null
     tiling=NoTile
     imageSize=<32 32>
     margin=<ltrb 0 0 0 0>
     outlineSettings {
      bUseBrushTransparency=false
      roundingType=FixedRadius
      width=1
      color {colorUseRule=UseColor_Specified, specifiedColor=<rgba 0.1198 0.1198 0.1198 0.5>}
      cornerRadii=<0 0 0 0>
     }
     tintColor {colorUseRule=UseColor_Specified, specifiedColor=<rgba 0.0423 0.0423 0.0423 1>}
     uVRegion {bIsValid=false, max=<0 0>, min=<0 0>}
    }
    padding=<ltrb 4 4 4 4>
   }
   slot {class=VerticalBoxSlot}
  }
  SoundListText {
   widgetClassPath=&TextBlock
   bIsVariable=true
   parent=Border_193
   widget=SoundListText
   properties {
    text=Animation List
    font {
     bForceMonospaced=false
     bMaterialIsStencil=false
     fontMaterial=null
     fontObject=&Burbank
     letterSpacing=0
     monospacedWidth=1
     size=15
     skewAmount=0
     typefaceFontName=Burbank-Black
     outlineSettings {
      bApplyOutlineToDropShadows=false
      bMiteredCorners=false
      bSeparateFillAlpha=false
      outlineMaterial=null
      outlineSize=0
      outlineColor=<rgba 0 0 0 1>
     }
    }
    shadowColorAndOpacity=<rgba 0 0 0 1>
   }
   slot {class=BorderSlot, properties {horizontalAlignment=HAlign_Center, padding=<ltrb 4 4 4 4>}}
  }
  EditorUtilityScrollBox_103 {
   widgetClassPath=&EditorUtilityScrollBox
   parent=VerticalBox_59
   widget=EditorUtilityScrollBox_103
   properties {
    clipping=Inherit
    widgetBarStyle {
     thickness=8
     draggedThumbImage {
      bIsDynamicallyLoaded=false
      drawAs=RoundedBox
      imageType=NoImage
      mirroring=NoMirror
      resourceName=None
      resourceObject=null
      tiling=NoTile
      imageSize=<0 0>
      margin=<ltrb 0 0 0 0>
      outlineSettings {
       bUseBrushTransparency=false
       roundingType=FixedRadius
       width=0
       color {colorUseRule=UseColor_Specified, specifiedColor=<rgba 0 0 0 0>}
       cornerRadii=<4 4 4 4>
      }
      tintColor {colorUseRule=UseColor_Specified, specifiedColor=<rgba 0.2159 0.2159 0.2159 1>}
      uVRegion {bIsValid=false, max=<0 0>, min=<0 0>}
     }
     horizontalBackgroundImage {
      bIsDynamicallyLoaded=false
      drawAs=NoDrawType
      imageType=NoImage
      mirroring=NoMirror
      resourceName=None
      resourceObject=null
      tiling=NoTile
      imageSize=<0 0>
      margin=<ltrb 0 0 0 0>
      outlineSettings {
       bUseBrushTransparency=false
       roundingType=HalfHeightRadius
       width=0
       color {colorUseRule=UseColor_Specified, specifiedColor=<rgba 0 0 0 0>}
       cornerRadii=<0 0 0 0>
      }
      tintColor {colorUseRule=UseColor_Specified, specifiedColor=<rgba 1 1 1 1>}
      uVRegion {bIsValid=false, max=<0 0>, min=<0 0>}
     }
     horizontalBottomSlotImage {
      bIsDynamicallyLoaded=false
      drawAs=NoDrawType
      imageType=NoImage
      mirroring=NoMirror
      resourceName=None
      resourceObject=null
      tiling=NoTile
      imageSize=<0 0>
      margin=<ltrb 0 0 0 0>
      outlineSettings {
       bUseBrushTransparency=false
       roundingType=HalfHeightRadius
       width=0
       color {colorUseRule=UseColor_Specified, specifiedColor=<rgba 0 0 0 0>}
       cornerRadii=<0 0 0 0>
      }
      tintColor {colorUseRule=UseColor_Specified, specifiedColor=<rgba 1 1 1 1>}
      uVRegion {bIsValid=false, max=<0 0>, min=<0 0>}
     }
     horizontalTopSlotImage {
      bIsDynamicallyLoaded=false
      drawAs=NoDrawType
      imageType=NoImage
      mirroring=NoMirror
      resourceName=None
      resourceObject=null
      tiling=NoTile
      imageSize=<0 0>
      margin=<ltrb 0 0 0 0>
      outlineSettings {
       bUseBrushTransparency=false
       roundingType=HalfHeightRadius
       width=0
       color {colorUseRule=UseColor_Specified, specifiedColor=<rgba 0 0 0 0>}
       cornerRadii=<0 0 0 0>
      }
      tintColor {colorUseRule=UseColor_Specified, specifiedColor=<rgba 1 1 1 1>}
      uVRegion {bIsValid=false, max=<0 0>, min=<0 0>}
     }
     hoveredThumbImage {
      bIsDynamicallyLoaded=false
      drawAs=RoundedBox
      imageType=NoImage
      mirroring=NoMirror
      resourceName=None
      resourceObject=null
      tiling=NoTile
      imageSize=<0 0>
      margin=<ltrb 0 0 0 0>
      outlineSettings {
       bUseBrushTransparency=false
       roundingType=FixedRadius
       width=0
       color {colorUseRule=UseColor_Specified, specifiedColor=<rgba 0 0 0 0>}
       cornerRadii=<4 4 4 4>
      }
      tintColor {colorUseRule=UseColor_Specified, specifiedColor=<rgba 0.2159 0.2159 0.2159 1>}
      uVRegion {bIsValid=false, max=<0 0>, min=<0 0>}
     }
     normalThumbImage {
      bIsDynamicallyLoaded=false
      drawAs=RoundedBox
      imageType=NoImage
      mirroring=NoMirror
      resourceName=None
      resourceObject=null
      tiling=NoTile
      imageSize=<0 0>
      margin=<ltrb 0 0 0 0>
      outlineSettings {
       bUseBrushTransparency=false
       roundingType=FixedRadius
       width=0
       color {colorUseRule=UseColor_Specified, specifiedColor=<rgba 0 0 0 0>}
       cornerRadii=<4 4 4 4>
      }
      tintColor {colorUseRule=UseColor_Specified, specifiedColor=<rgba 0.0953 0.0953 0.0953 0>}
      uVRegion {bIsValid=false, max=<0 0>, min=<0 0>}
     }
     verticalBackgroundImage {
      bIsDynamicallyLoaded=false
      drawAs=NoDrawType
      imageType=NoImage
      mirroring=NoMirror
      resourceName=None
      resourceObject=null
      tiling=NoTile
      imageSize=<0 0>
      margin=<ltrb 0 0 0 0>
      outlineSettings {
       bUseBrushTransparency=false
       roundingType=HalfHeightRadius
       width=0
       color {colorUseRule=UseColor_Specified, specifiedColor=<rgba 0 0 0 0>}
       cornerRadii=<0 0 0 0>
      }
      tintColor {colorUseRule=UseColor_Specified, specifiedColor=<rgba 1 1 1 1>}
      uVRegion {bIsValid=false, max=<0 0>, min=<0 0>}
     }
     verticalBottomSlotImage {
      bIsDynamicallyLoaded=false
      drawAs=NoDrawType
      imageType=NoImage
      mirroring=NoMirror
      resourceName=None
      resourceObject=null
      tiling=NoTile
      imageSize=<0 0>
      margin=<ltrb 0 0 0 0>
      outlineSettings {
       bUseBrushTransparency=false
       roundingType=HalfHeightRadius
       width=0
       color {colorUseRule=UseColor_Specified, specifiedColor=<rgba 0 0 0 0>}
       cornerRadii=<0 0 0 0>
      }
      tintColor {colorUseRule=UseColor_Specified, specifiedColor=<rgba 1 1 1 1>}
      uVRegion {bIsValid=false, max=<0 0>, min=<0 0>}
     }
     verticalTopSlotImage {
      bIsDynamicallyLoaded=false
      drawAs=NoDrawType
      imageType=NoImage
      mirroring=NoMirror
      resourceName=None
      resourceObject=null
      tiling=NoTile
      imageSize=<0 0>
      margin=<ltrb 0 0 0 0>
      outlineSettings {
       bUseBrushTransparency=false
       roundingType=HalfHeightRadius
       width=0
       color {colorUseRule=UseColor_Specified, specifiedColor=<rgba 0 0 0 0>}
       cornerRadii=<0 0 0 0>
      }
      tintColor {colorUseRule=UseColor_Specified, specifiedColor=<rgba 1 1 1 1>}
      uVRegion {bIsValid=false, max=<0 0>, min=<0 0>}
     }
    }
   }
   slot {class=VerticalBoxSlot, properties {padding=<ltrb 2 2 2 2>}}
  }
  OutputText {
   widgetClassPath=&TextBlock
   bIsVariable=true
   parent=EditorUtilityScrollBox_103
   widget=OutputText
   properties {
    visibility=SelfHitTestInvisible
    font {
     bForceMonospaced=false
     bMaterialIsStencil=false
     fontMaterial=null
     fontObject=/LauncherSocial/BrutalType
     letterSpacing=0
     monospacedWidth=1
     size=15
     skewAmount=0
     typefaceFontName=Regular
     outlineSettings {
      bApplyOutlineToDropShadows=false
      bMiteredCorners=false
      bSeparateFillAlpha=false
      outlineMaterial=null
      outlineSize=0
      outlineColor=<rgba 0 0 0 1>
     }
    }
    shadowColorAndOpacity=<rgba 0 0 0 1>
    shadowOffset=<2 1>
   }
   slot {class=ScrollBoxSlot}
  }
  MainBox {
   widgetClassPath=&EditorUtilityScrollBox
   bIsVariable=true
   parent=VerticalBox_59
   widget=MainBox
   properties {alwaysShowScrollbar=true, alwaysShowScrollbarTrack=true}
   slot {class=VerticalBoxSlot, properties {size {sizeRule=Fill, value=1}}}
  }
 ]
}
