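# Editor utility widget for audio developers that builds MetaSound presets, a Sound Library and animation notifies for melee weapons.
# It spawns the weapon actor to read its Mantis techniques, name-matches them against a data table to get sound tags, and lists actions with their montages.
# A looping timer mirrors selected SoundWaves into the list; assigned waves feed preset builders, whose assets fill the library.
# Library creation also links it to the weapon item definitions.
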EditorUtilityWidgetBlueprint /Game/Sounds/Fort_Audio_Blueprints/Utilities/EditorUtilityWidgets/Metasounds/Melee/EUW_MetasoundPresetCreator_MeleeWeapons
graphs [
 ASsetSelectionTimer {nodes [
   ASsetSelectionTimer() {is=title K2Node_FunctionEntry}
   > n1 = GetSelectedAssetsOfClass(AssetClass=/Script/Engine.SoundWave)
   > n2 = For Each Loop(Array:Array of Object References=@n1) -> (Array Element:Object Reference)
   > n3 = Cast To MetaSoundSource!(execute, Object=@n2) -> (then, CastFailed, AsMetaSound Source:Meta Sound Source Object Reference) {is=title K2Node_DynamicCast}
   n4 = Sequence(execute=@n3.CastFailed)
   > n5 = Branch(Condition=Contains Item(
     TargetArray:Array of Object References=Get SelectedAssets() -> (SelectedAssets:Array of Object References)
     ItemToFind:"Object Reference (by ref)"=@n2
    ))
   n6 = Add Unique(execute=@n5.else, TargetArray:Array of Object References=Get SelectedAssets(), NewItem:"Object Reference (by ref)"=@n2)
   > AddEntryUpdate!(
    execute
    self:EUW Audio Weapons Sound Name List View Object Reference=Get EUW_Audio_Weapons_SoundName_ListView() -> (EUW_Audio_Weapons_SoundName_ListView:EUW Audio Weapons Sound Name List View Object Reference)
    Object:Object Reference=@n2
    AssetIndex:"Integer (by ref)"=@n6
   ) -> (then)
   n7 = For Each Loop(Exec=@n4.then_1, Array=Get SelectedAssets())
   > n8 = Branch(Condition=Contains Item(TargetArray=@n1, ItemToFind=@n7))
   Remove Item!(execute=@n8.else, TargetArray:Array of Object References=Get SelectedAssets(), Item:"Object Reference (by ref)"=@n7) -> (then, ReturnValue:bool)
   > RemoveEntryUpdate!(
    execute
    self:EUW Audio Weapons Sound Name List View Object Reference=Get EUW_Audio_Weapons_SoundName_ListView()
    Object:Object Reference=@n7
    AssetIndex:"Integer (by ref)"=@n7.Array Index
   ) -> (then)
  ]}
 AssignWavs {nodes [
   n1 = AssignWavs!() -> (then, Target:EUW Audio Weapons Action Name Widget Object Reference, VariableName:Name) {is=title K2Node_FunctionEntry}
   > n2 = For Each Loop(Array:Array of Widget Object References=GetAllChildren(self=Get ScrollBox!(self:EUW Audio Weapons Action Name Widget Object Reference=@n1) -> (ScrollBox:Editor Utility Scroll Box Object Reference))) -> (Array Element:Widget Object Reference)
   n3 = Get CurrentEvaluatedWaves() -> (CurrentEvaluatedWaves:Array of Object References)
   n4 = Branch(execute=@n2.Completed, Condition=Is Not Empty!(TargetArray:Array of Object References=@n3) -> (ReturnValue:bool))
   n5 = Get VariableName() -> (VariableName:Name=None)
   > n6 = Create MetaSound Literal From AudioParameter!(
    execute
    self:Meta Sound Builder Subsystem Object Reference=Get MetaSoundBuilderSubsystem() -> (ReturnValue:Meta Sound Builder Subsystem Object Reference) {is=title K2Node_GetEngineSubsystem}
    Param:"Audio Parameter Structure (by ref)"=ObjectArrayToAudioParameter!(Name:Name=@n5, Objects:Array of Object References=@n3) -> (ReturnValue:Audio Parameter Structure)
   ) -> (then, ReturnValue:Meta Sound Literal Structure)
   > n7 = SetGraphInputDefault!(
    execute
    self:Meta Sound Builder Base Object Reference=Get MetaSoundPresetBuilder() -> (MetaSoundPresetBuilder:Meta Sound Source Builder Object Reference)
    InputName:Name=@n5
    Literal:"Meta Sound Literal Structure (by ref)"=@n6
    PageName:Name
   ) -> (Succeeded:Exec, Failed:Exec)
   Return Node!(execute=@n7.Failed, bSuccess:bool=false)
   > Return Node(bSuccess=true)
   Return Node(execute=@n4.else, bSuccess=false)
   n8 = Cast To EUW_Audio_Weapons_SoundNameObject!(execute=@n2.LoopBody, Object=@n2) -> (then, CastFailed, AsEUW Audio Weapons Sound Name Object:EUW Audio Weapons Sound Name Object Object Reference) {is=title K2Node_DynamicCast}
   > Add Unique(
    TargetArray=Get CurrentEvaluatedWaves()
    NewItem=Get SoundObjectRef!(self:EUW Audio Weapons Sound Name Object Object Reference=@n8) -> (SoundObjectRef:Object Reference)
   )
  ]}
 CreateAndAddSoundLibrary {nodes [
   CreateAndAddSoundLibrary() {is=title K2Node_FunctionEntry}
   n1 = Get EUW_MetaSoundAssetPath() -> (EUW_MetaSoundAssetPath:EUW Meta Sound Asset Path Object Reference)
   n2 = Get Text(self=Get AssetNameInput!(self:EUW Meta Sound Asset Path Object Reference=@n1) -> (AssetNameInput:Editor Utility Editable Text Box Object Reference))
   > n3 = Branch(Condition=TextIsEmpty(InText=@n2))
   n4 = GetAssetTools(execute=@n3.else)
   n5 = Get PathInput!(self:EUW Meta Sound Asset Path Object Reference=@n1) -> (PathInput:Editor Utility Editable Text Box Object Reference)
   > n6 = CreateAsset(
    self=@n4
    AssetName=Append(A=SL_Melee_, B="To String (Text)"(InText=@n2))
    PackagePath="To String (Text)"(InText=Get Text(self=@n5))
    AssetClass=/Script/SoundLibrary
    CallingContext=Weapon
   )
   > n7 = Cast To SoundLibrary!(execute, Object=@n6) -> (then, CastFailed, AsSound Library:Sound Library Object Reference) {is=title K2Node_DynamicCast}
   > n8 = Set SoundLibrary(SoundLibrary:Sound Library Object Reference=@n7) -> (Output_Get:Sound Library Object Reference)
   > SetEditorProperty(
    Object=@n8
    PropertyName=Parent
    PropertyValue:Sound Library Object Reference=Get SoundLibraryParent() -> (SoundLibraryParent:Sound Library Object Reference)
   )
   > n9 = For Each Loop(Array:Array of Fort Weapon Item Definition Object References=Get WeaponId() -> (WeaponId:Array of Fort Weapon Item Definition Object References)) -> (Array Element:Fort Weapon Item Definition Object Reference)
   > n10 = GetItemData!(execute, self:Item Definition Base Object Reference=@n9) -> (Found:Exec, NotFound:Exec, ItemData:"Fortnite Weapon Sound Libraries Item Data Structure (by ref)")
   > n11 = RemoveItemData!(
    execute
    self:Item Definition Base Object Reference=@n9
    StructType:Script Struct Object Reference=/Script/FortniteGame.FortItemComponentData_WeaponSoundLibraries
   ) -> (Found:Exec, NotFound:Exec)
   n12 = Get!(execute:Exec=[@n11.Found, @n10.NotFound]) -> (then:Exec, else:Exec, SoundLibrary:Sound Library Object Reference)
   > SetItemData!(
    execute
    self:Item Definition Base Object Reference=@n9
    ItemData:Fortnite Weapon Sound Libraries Item Data Structure=Make Fortnite Weapon Sound Libraries Item Data!(
     WeaponSoundLibraries:Array of Sound Library Object References=Make Array("[0]":Sound Library Object Reference=@n12) -> (Array:Array of Sound Library Object References)
     WeaponOwnerSoundLibraries:Array of Sound Library Object References
     WeaponCollectionTag:Gameplay Tag Container Structure=(GameplayTags=)
    ) -> (FortItemComponentData_WeaponSoundLibraries:Fortnite Weapon Sound Libraries Item Data Structure) {is=title K2Node_MakeStruct}
   ) -> (then)
   Break Fortnite Weapon Sound Libraries Item Data!(FortItemComponentData_WeaponSoundLibraries:"Fortnite Weapon Sound Libraries Item Data Structure (by ref)"=@n10) {is=title K2Node_BreakStruct}
  ]}
 CreateAnimArray {nodes [
   n1 = CreateAnimArray!() -> (then, Anims:Map of Names to Anim Montage Object References) {is=title K2Node_FunctionEntry}
   > n2 = "For Each Loop (Map)"!(execute:Exec, MapPin:Map of Names to Anim Montage Object References=@n1, BreakPin:Exec) -> (then:Exec, KeyPin:Name, ValuePin:Anim Montage Object Reference, CompletedPin:Exec)
   n3 = Get TempArray() -> (TempArray:Array of Anim Montage Object References)
   > Add Unique(TargetArray:Array of Anim Montage Object References=@n3, NewItem:"Anim Montage Object Reference (by ref)"=@n2.ValuePin)
   Return Node!(execute=@n2.CompletedPin, Montages:Array of Anim Montage Object References=@n3)
  ]}
 CreateNotifies {nodes [
   n1 = CreateNotifies!() -> (then, Target:EUW Melee Audio Animation Name List View Object Reference, Name:"Text (by ref)") {is=title K2Node_FunctionEntry}
   > n2 = For Each Loop with Break!(
    Exec:Exec
    Array:Array of EUW Melee Animation Selection Object References=Get AnimationNameWidgets!(self:EUW Melee Audio Animation Name List View Object Reference=@n1) -> (AnimationNameWidgets:Array of EUW Melee Animation Selection Object References)
    Break:Exec=@n4.then_0
   ) -> (LoopBody:Exec, Array Element:EUW Melee Animation Selection Object Reference, Array Index:int, Completed:Exec)
   Return Node!(execute=@n2.Completed)
   n3 = Get PresetText!(self:EUW Melee Animation Selection Object Reference=@n2) -> (PresetText:Text Object Reference)
   > Branch(Condition="Equal Exactly (Text)"!(A:"Text (by ref)"=Get Text(self:Text Object Reference=@n3), B:"Text (by ref)"=Get Name() -> (Name:"Text (by ref)")) -> (ReturnValue:bool))
   > n4 = Sequence()
   n5 = For Each Loop(
    Exec=@n4.then_1
    Array:Array of Widget Object References=GetAllChildren(self=Get ScrollBox(self:EUW Melee Animation Selection Object Reference=@n2))
   ) -> (Array Element:Widget Object Reference)
   > n6 = Cast To EUW_AnimationEntry!(execute, Object=@n5) -> (then, CastFailed, AsEUW Animation Entry:EUW Animation Entry Object Reference) {is=title K2Node_DynamicCast}
   n7 = Find(
    TargetMap:Map of Strings to Sound Library Tag Structures=Get SoundLibraryEvents() -> (SoundLibraryEvents:Map of Strings to Sound Library Tag Structures)
    Key="To String (Text)"(InText=Get Name())
   ) -> (Value:Sound Library Tag Structure)
   > Branch(Condition=AND Boolean(A=@n7.ReturnValue, B=IsGameplayTagValid!(GameplayTag:Gameplay Tag Structure=@n7) -> (ReturnValue:bool)))
   > AddNotify!(execute, self:EUW Animation Entry Object Reference=@n6, SoundLibraryEvent:Sound Library Tag Structure=@n7) -> (then)
  ]}
 EventGraph {nodes [
   "Event on Make Preset Button (EUW_MetaSoundAssetPath)"() {is=title K2Node_ComponentBoundEvent}
   > n1 = Get()
   StartMetasoundBuilder(execute=[@n2.then, @n1.then])
   > FillSoundLibrary()
   n2 = CreateAndAddSoundLibrary(execute=@n1.else) {is=title K2Node_CallFunction}
   "On Clicked (SoundLibButton)"() {is=title K2Node_ComponentBoundEvent}
   > CreateAndAddSoundLibrary()
   n3 = PropertyChanged!() -> (OutputDelegate, then, PropertyName:Name) {is=title K2Node_CustomEvent}
   > n4 = Switch on Name!(execute:Exec, Selection:Name=@n3.PropertyName) -> (Default:Exec, MantisData:Exec, SoundLibraryEvents:Exec, WeaponId:Exec, MeleeNameMatches:Exec) {is=title K2Node_SwitchName, pinNames [MantisData, SoundLibraryEvents, WeaponId, MeleeNameMatches]}
   UpdateMantisData(execute=@n4.MantisData)
   UpdateSoundMap(execute=[@n4.MeleeNameMatches, @n4.SoundLibraryEvents])
   UpdateWeaponId(execute=@n4.WeaponId)
   Event PreConstruct()
   > SetObject(self=Get DetailsView_135() -> (DetailsView_135:Details View Object Reference), NewObject=Self-Reference())
   Event Construct()
   > Bind Event to On Property Changed(self:Property View Base Object Reference=Get DetailsView_135(), Delegate=Create Event()) {is=title K2Node_AddDelegate}
   > n5 = Branch(Condition=Is Valid Timer Handle!(Handle:Timer Handle Structure=Get SelectionTimerHandle() -> (SelectionTimerHandle:Timer Handle Structure)) -> (ReturnValue:bool))
   n6 = Set Timer by Event(execute=@n5.else, Delegate=Create Event(), Time=0.2, bLooping=true)
   > Set SelectionTimerHandle(SelectionTimerHandle:Timer Handle Structure=@n6) -> (Output_Get:Timer Handle Structure)
  ]}
 FillSoundLibrary {nodes [
   FillSoundLibrary() {is=title K2Node_FunctionEntry}
   > SetEditorProperty(
    Object=Get SoundLibrary() -> (SoundLibrary:Sound Library Object Reference)
    PropertyName=Sounds
    PropertyValue:Map of Gameplay Tag Structures to Sound Base Soft Object References=Get SoundLibrarySounds() -> (SoundLibrarySounds:Map of Gameplay Tag Structures to Sound Base Soft Object References)
   )
  ]}
 SaveAssets {nodes [
   SaveAssets() {is=title K2Node_FunctionEntry}
   > n1 = For Each Loop(Array:Array of Meta Sound Source Object References=Get BuiltMetaSounds() -> (BuiltMetaSounds:Array of Meta Sound Source Object References)) -> (Array Element:Meta Sound Source Object Reference)
   SaveLoadedAsset(execute=@n1.Completed, AssetToSave=Get SoundLibrary(), bOnlyIfIsDirty=true)
   SaveLoadedAsset(execute=@n1.LoopBody, AssetToSave=@n1, bOnlyIfIsDirty=true)
  ]}
 StartMetasoundBuilder {nodes [
   StartMetasoundBuilder() {is=title K2Node_FunctionEntry}
   n1 = Get PathInput(self=Get EUW_MetaSoundAssetPath())
   > n2 = Branch(Condition=TextIsEmpty(InText=Get Text(self:Text Box Object Reference=@n1)))
   Clear!(execute=@n2.else, TargetMap:Map of Gameplay Tag Structures to Sound Base Soft Object References=Get SoundLibrarySounds()) -> (then) {is=title K2Node_CallFunction}
   > Clear(TargetArray:Array of Meta Sound Source Object References=Get BuiltMetaSounds()) {is=title K2Node_CallArrayFunction}
   > n3 = For Each Loop(Array:Array of EUW Audio Weapons Action Name Widget Object References=Get WeaponActionNameWidgets!(self:EUW Audio Weapons Sound Name List View Object Reference=Get EUW_Audio_Weapons_SoundName_ListView()) -> (WeaponActionNameWidgets:Array of EUW Audio Weapons Action Name Widget Object References)) -> (Array Element:EUW Audio Weapons Action Name Widget Object Reference)
   n4 = Get ScrollBox(self:EUW Audio Weapons Action Name Widget Object Reference=@n3)
   > Branch(Condition="integer > integer"(A=GetChildrenCount!(self:Panel Widget Object Reference=@n4) -> (ReturnValue:int=0)))
   > Set CurrentWidget(CurrentWidget:EUW Audio Weapons Action Name Widget Object Reference=@n3) -> (Output_Get:EUW Audio Weapons Action Name Widget Object Reference)
   > n5 = Sequence()
   n6 = Get Text(self:Text Object Reference=Get PresetText(self:EUW Audio Weapons Action Name Widget Object Reference=Get CurrentWidget() -> (CurrentWidget:EUW Audio Weapons Action Name Widget Object Reference)))
   > Set CurrentName(CurrentName:Text=@n6) -> (Output_Get:Text)
   n7 = "To String (Text)"(InText=@n6)
   > n8 = Branch(Condition=OR Boolean!(A=Contains(Substring=" "), B=Contains(SearchIn=@n7, Substring="("), C:bool=Contains(Substring=")")) -> (ReturnValue))
   n9 = Cast To MetaSoundDocumentInterface!(execute=@n8.else, Object=Get MetasoundParent() -> (MetasoundParent:Meta Sound Source Object Reference)) -> (then, CastFailed, AsMetaSound Document Interface:Meta Sound Document Interface Interface) {is=title K2Node_DynamicCast}
   n10 = "To String (Text)"(InText=Get CurrentName() -> (CurrentName:Text))
   > n11 = CreateSourcePresetBuilder!(
    execute
    self:Meta Sound Builder Subsystem Object Reference=Get MetaSoundBuilderSubsystem()
    BuilderName:Name=String To Name(InString=Append(A=@n10, B=_MeleeBuilder))
    ReferencedSourceClass:"Meta Sound Document Interface Interface (by ref)"=@n9
   ) -> (Succeeded:Exec, Failed:Exec, ReturnValue:Meta Sound Source Builder Object Reference)
   > Set MetaSoundPresetBuilder(MetaSoundPresetBuilder:Meta Sound Source Builder Object Reference=@n11) -> (Output_Get:Meta Sound Source Builder Object Reference)
   > n12 = AssignWavs!(execute, self, Target:EUW Audio Weapons Action Name Widget Object Reference=Get CurrentWidget(), VariableName:Name=Sounds.Swings) -> (then, bSuccess:bool) {is=title K2Node_CallFunction}
   > Branch(Condition=@n12)
   n13 = Get EUW_MetaSoundAssetPath()
   n14 = Get Text(self:Text Box Object Reference=Get PathInput(self=@n13))
   n15 = Get LicensedAudio() -> (LicensedAudio:bool)
   n16 = Get Text(self=Get AssetNameInput(self=@n13))
   n17 = Get CurrentName()
   > n18 = BuildToAsset!(
    execute
    self:Meta Sound Editor Subsystem Object Reference=Get MetaSoundEditorSubsystem() -> (ReturnValue:Meta Sound Editor Subsystem Object Reference) {is=title K2Node_GetEditorSubsystem}
    InBuilder:Meta Sound Builder Base Object Reference=Get MetaSoundPresetBuilder()
    Author:String=SoundBaseAssetAction
    AssetName:String=Append!(
     A=SelectString(A=IP_MSS_Weapon_, B=MSS_Weapon_, bPickA=@n15)
     B="To String (Text)"(InText=@n16)
     C:String=_
     D:String="To String (Text)"(InText=@n17)
    ) -> (ReturnValue)
    PackagePath:String="To String (Text)"(InText=@n14)
    TemplateSoundWave:Sound Wave Object Reference
   ) -> (Succeeded:Exec, Failed:Exec, ReturnValue:Meta Sound Document Interface Interface)
   n19 = "To String (Text)"(InText=@n14)
   PrintString(
    execute=@n18.Failed
    InString="To String (Text)"(InText=Format Text!(Format="LOCTEXT(\"Failed to build {string}, please close and restart the tool.\")", string:String=@n19) -> (Result))
   )
   > Return Node()
   n20 = Cast To MetaSoundSource(execute=@n18.Succeeded, Object:Meta Sound Document Interface Interface=@n18)
   > Add Unique(TargetArray:Array of Meta Sound Source Object References=Get BuiltMetaSounds(), NewItem:"Meta Sound Source Object Reference (by ref)"=@n20)
   n21 = Find(TargetMap=Get SoundLibraryEvents(), Key="To String (Text)"(InText=Get CurrentName()))
   n22 = Contains!(
    TargetMap:Map of Gameplay Tag Structures to Sound Base Soft Object References=Get SoundLibrarySounds()
    Key:"Gameplay Tag Structure (by ref)"=@n21
   ) -> (ReturnValue)
   > Branch(Condition=AND Boolean(A=IsGameplayTagValid(GameplayTag=@n21), B=NOT Boolean(A=@n22)))
   > Add!(
    execute
    TargetMap:Map of Gameplay Tag Structures to Sound Base Soft Object References=Get SoundLibrarySounds()
    Key:"Gameplay Tag Structure (by ref)"=@n21
    Value:"Sound Base Soft Object Reference (by ref)"=Make Soft Reference!(Input:Meta Sound Source Object Reference=@n20) -> (Output:Meta Sound Source Soft Object Reference)
   ) -> (then) {is=title K2Node_CallFunction}
   CreateNotifies!(
    execute=@n5.then_1
    self
    Target:EUW Melee Audio Animation Name List View Object Reference=Get EUW_MeleeAudio_AnimationName_ListView() -> (EUW_MeleeAudio_AnimationName_ListView:EUW Melee Audio Animation Name List View Object Reference)
    Name:"Text (by ref)"=Get CurrentName()
   ) -> (then) {is=title K2Node_CallFunction}
   GetPath!(self:EUW Meta Sound Asset Path Weapon Object Reference=Get EUW_MetaSoundAssetPath_Weapon() -> (EUW_MetaSoundAssetPath_Weapon:EUW Meta Sound Asset Path Weapon Object Reference)) -> (PathText:Text, bIsEmpty:bool)
  ]}
 TechniqueNameCheck {nodes [
   TechniqueNameCheck!() -> (then, Name:Name) {is=title K2Node_FunctionEntry}
   > n1 = For Each Loop with Break(Array:Array of Names=Get DataTableNames() -> (DataTableNames:Array of Names), Break=@n5.then) -> (Array Element:Name)
   Return Node!(
    execute=@n1.Completed
    NameMatchData1:S Melee Sound Library Name Match Structure=Get NameMatchData() -> (NameMatchData:S Melee Sound Library Name Match Structure)
   )
   > n2 = Get Data Table Row!(execute:Exec, DataTable:Data Table Object Reference=Get DataTable() -> (DataTable:Data Table Object Reference), RowName:Name=@n1) -> (then:Exec, RowNotFound:Exec, ReturnValue:S Melee Sound Library Name Match Structure)
   > n3 = For Each Loop with Break(Array:Array of Strings=Break S Melee Sound Library Name Match!(S_MeleeSoundLibraryNameMatch:"S Melee Sound Library Name Match Structure (by ref)"=@n2) -> (
     SoundLibraryEvent_4_#1:Sound Library Tag Structure
     Names_6_#2:Array of Strings
     SoundMatch_9_#3:String
    ) {is=title K2Node_BreakStruct}.Names_6_#2) -> (Array Element:String)
   n4 = Get Name() -> (Name:Name=None)
   > Branch(Condition=Contains!(SearchIn:String="To String (Name)"(InName=@n4), Substring:String=@n3, bUseCase:bool, bSearchFromEnd:bool=true) -> (ReturnValue))
   > n5 = Set NameMatchData(NameMatchData:S Melee Sound Library Name Match Structure=@n2) -> (Output_Get:S Melee Sound Library Name Match Structure)
   Get SoundlibraryEvent() -> (SoundlibraryEvent:Sound Library Tag Structure)
  ]}
 UpdateMantisData {nodes [
   UpdateMantisData() {is=title K2Node_FunctionEntry}
   > n1 = Sequence()
   n2 = For Each Loop(
    Exec=@n1.then_1
    Array:Array of Fort Mantis Technique Data Structures=Get Techniques!(self:Fort Mantis Data Object Reference=Get MantisData() -> (MantisData:Fort Mantis Data Object Reference)) -> (Techniques:Array of Fort Mantis Technique Data Structures)
   ) -> (Array Element:Fort Mantis Technique Data Structure)
   Set OldMantisData(execute=@n2.Completed, OldMantisData:Fort Mantis Data Object Reference=Get MantisData()) -> (Output_Get:Fort Mantis Data Object Reference)
   > UpdateSoundMap()
   n3 = Break Fort Mantis Technique Data!(FortMantisTechniqueData:"Fort Mantis Technique Data Structure (by ref)"=@n2) -> (Name:Name) {is=title K2Node_BreakStruct}
   n4 = TechniqueNameCheck!(execute=@n2.LoopBody, self, Name:Name=@n3) -> (then, NameMatchData1:S Melee Sound Library Name Match Structure) {is=title K2Node_CallFunction}
   n5 = Break S Melee Sound Library Name Match(S_MeleeSoundLibraryNameMatch=@n4)
   n6 = MakeLiteralGameplayTag!(Value:Gameplay Tag Structure) -> (ReturnValue:Gameplay Tag Structure)
   > Add(
    TargetMap:Map of Strings to Sound Library Tag Structures=Get SoundLibraryEvents()
    Key:"String (by ref)"="To String (Name)"(InName=@n3)
    Value:"Sound Library Tag Structure (by ref)"=Select(
     Option 0:"Sound Library Tag Structure (by ref)"=MakeSoundLibraryTag!(Tag:"Gameplay Tag Structure (by ref)"=@n6) -> (ReturnValue:Sound Library Tag Structure)
     Option 1:"Sound Library Tag Structure (by ref)"=@n5
     Index=IsGameplayTagValid(GameplayTag=@n5)
    ) -> (ReturnValue:"Sound Library Tag Structure (by ref)")
   )
   > Add(
    TargetMap:Map of Names to Strings=Get AudioNameMap() -> (AudioNameMap:Map of Names to Strings)
    Key:"Name (by ref)"=@n3
    Value:"String (by ref)"=@n5.SoundMatch_9_#3
   )
   n7 = For Each Loop(Exec=@n1.then_0, Array=Get Techniques(self=Get OldMantisData() -> (OldMantisData:Fort Mantis Data Object Reference)))
   > Remove!(
    execute
    TargetMap:Map of Strings to Sound Library Tag Structures=Get SoundLibraryEvents()
    Key:"String (by ref)"="To String (Name)"(InName=Break Fort Mantis Technique Data(FortMantisTechniqueData=@n7))
   ) -> (then, ReturnValue:bool)
  ]}
 UpdateSoundMap {nodes [
   UpdateSoundMap() {is=title K2Node_FunctionEntry}
   > Clear(TargetMap:Map of Strings to Strings=Get StringToStringMap() -> (StringToStringMap:Map of Strings to Strings)) {is=title K2Node_CallFunction}
   > ResetActionWidgets(self:EUW Audio Weapons Sound Name List View Object Reference=Get EUW_Audio_Weapons_SoundName_ListView())
   > Reset Widgets!(execute, self:EUW Melee Audio Animation Name List View Object Reference=Get EUW_MeleeAudio_AnimationName_ListView(), HardReset:bool=true) -> (then)
   > n1 = "For Each Loop (Map)"(MapPin:Map of Strings to Sound Library Tag Structures=Get SoundLibraryEvents()) -> (KeyPin:String, ValuePin:Sound Library Tag Structure)
   > n2 = Set TechniqueName(TechniqueName:String=@n1) -> (Output_Get:String)
   > Add(
    TargetMap:Map of Strings to Strings=Get StringToStringMap()
    Key:"String (by ref)"=@n2
    Value=Find(TargetMap:Map of Names to Strings=Get AudioNameMap(), Key:"Name (by ref)"=String To Name(InString=@n1)) -> (Value:String)
   )
   > SetupWidgetActions!(
    execute
    self:EUW Audio Weapons Sound Name List View Object Reference=Get EUW_Audio_Weapons_SoundName_ListView()
    ActionName:String=Get TechniqueName() -> (TechniqueName:String)
   ) -> (then)
   > n3 = For Each Loop with Break(Array:Array of Fort Mantis Technique Data Structures=Get Techniques(self=Get MantisData()), Break=@n9.then) -> (Array Element:Fort Mantis Technique Data Structure)
   n4 = Get!(execute=@n3.Completed) -> (then, else, FoundWidget:bool)
   > Set FoundWidget(FoundWidget:bool=false) -> (Output_Get:bool)
   SetUpWidgetActions!(
    execute=@n4.else
    self:EUW Melee Audio Animation Name List View Object Reference=Get EUW_MeleeAudio_AnimationName_ListView()
    ActionName:String=Get TechniqueName()
    Animations:Array of Anim Sequence Base Object References
   ) -> (then)
   Clear(
    execute=@n3.LoopBody
    TargetArray:Array of Anim Montage Object References=Get Montages() -> (Montages:Array of Anim Montage Object References)
   ) {is=title K2Node_CallArrayFunction}
   n5 = Break Fort Mantis Technique Data!(FortMantisTechniqueData=@n3) -> (Name, Montage:Anim Montage Object Reference, DynamicMontages:Map of Names to Anim Montage Object References)
   > n6 = Branch(Condition=Is Not Empty!(TargetMap:Map of Names to Anim Montage Object References=@n5.DynamicMontages) -> (ReturnValue:bool) {is=title K2Node_CallFunction})
   > n7 = CreateAnimArray!(execute, self, Anims:Map of Names to Anim Montage Object References=@n5.DynamicMontages) -> (then, Montages:Array of Anim Montage Object References) {is=title K2Node_CallFunction}
   > n8 = Set Montages(Montages:Array of Anim Montage Object References=@n7) -> (Output_Get:Array of Anim Montage Object References)
   Branch(execute=[@n11.then, @n8.then, @n10.Is Not Valid], Condition="Equal (Name)"(A=@n5, B:String=Get TechniqueName()))
   > SetUpWidgetActions(self=Get EUW_MeleeAudio_AnimationName_ListView(), ActionName=Get TechniqueName(), Animations=Get Montages())
   > n9 = Set FoundWidget(FoundWidget=true)
   n10 = Is Valid(exec=@n6.else, InputObject=@n5.Montage)
   > n11 = Add(TargetArray:Array of Anim Montage Object References=Get Montages(), NewItem:"Anim Montage Object Reference (by ref)"=@n5.Montage) {is=title K2Node_CallArrayFunction}
   SetStringToStringMap!(
    execute=@n1.CompletedPin
    self:EUW Audio Weapons Sound Name List View Object Reference=Get EUW_Audio_Weapons_SoundName_ListView()
    TMapStringToStringInput:Map of Strings to Strings=Get StringToStringMap()
   ) -> (then)
   > AssignWavs(VariableName=None)
  ]}
 UpdateWeaponId {nodes [
   UpdateWeaponId() {is=title K2Node_FunctionEntry}
   n1 = "Get (a copy)"(Array:Array of Fort Weapon Item Definition Object References=Get WeaponId()) -> (Output:Fort Weapon Item Definition Object Reference)
   > Is Valid(InputObject=@n1)
   > n2 = Cast To B_Mantis_Weapon_Parent Class!(
    execute:Exec
    Class:Object Class Reference=GetWeaponActorClass!(self:Fort Weapon Item Definition Object Reference=@n1) -> (ReturnValue:Fort Weapon Class Reference)
   ) -> (then:Exec, CastFailed:Exec, AsB Mantis Weapon Parent:B Mantis Weapon Parent Class Reference) {is=title K2Node_ClassDynamicCast}
   > n3 = SpawnActor!(
    execute:Exec
    WorldContextObject:Object Reference
    Class:Actor Class Reference=@n2
    SpawnTransform:Transform=MakeTransform(Rotation="0, 0, 0")
    CollisionHandlingOverride:ESpawnActorCollisionHandlingMethod Enum=AlwaysSpawn
    TransformScaleMethod:ESpawnActorScaleMethod Enum=MultiplyWithRoot
    Owner:Actor Object Reference
    IdleFXAttachSocket:Name=idle_fx
    SwingFXSocket:Name=SwingFX
    Instigator:Pawn Object Reference
   ) -> (then:Exec, ReturnValue:B Mantis Weapon Parent Object Reference) {is=title K2Node_SpawnActorFromClass}
   n4 = Get MantisData!(self:Fort Mantis Weapon Component Object Reference=Get FortMantisWeapon!(self:B Mantis Weapon Parent Object Reference=@n3) -> (FortMantisWeapon:Fort Mantis Weapon Component Object Reference)) -> (MantisData)
   > n5 = Is Valid(InputObject=@n4)
   n6 = Show Message Dialog(
    execute=@n5.Is Not Valid
    Title=MakeLiteralText(Value="LOCTEXT(\"No Mantis data found\")")
    Message=MakeLiteralText(Value="LOCTEXT(\"Unable to find mantis data on melee weapon, please either reach out to the design team for this weapon and ask if that is intended, or create your own entries by filling out the sound library.\")")
    MessageType=Ok
    DefaultValue=No
    MessageCategory=Warning
   )
   Destroy Actor(execute=[@n6.then, @n7.then], self=@n3)
   > UpdateMantisData() {is=title K2Node_CallFunction}
   n7 = Set MantisData(execute=@n5.Is Valid, MantisData:Fort Mantis Data Object Reference=@n4) -> (Output_Get:Fort Mantis Data Object Reference)
  ]}
]
variables [
 AudioNameMap {type="map<string>", default {}}
 BuiltMetaSounds {type="array<MetaSoundSource>", default []}
 DataTable {type=DataTable, default=&/DT_MeleeWeapon_NameMatch}
 DataTableNames {type="array<string>", default [ComboOne,ComboTwo,ComboThree,ComboFour,CopmboFive,ComboSix,ComboSeven,ComboEight,StartSlam,GroundSlam,Heavy]}
 EUW_MetaSoundAssetPath_Weapon {type=EUW_MetaSoundAssetPath_Weapon_C, default=null}
 FoundWidget {type=boolean}
 LicensedAudio {type=boolean}
 MantisData {type=FortMantisData, default=null}
 MeleeNameMatches {type="array<S_MeleeSoundLibraryNameMatch>", default []}
 MetasoundParent {type=MetaSoundSource, default=/FortMetasound/Weapons/MeleeWeapons/MSS_Weapon_Melee_Swing, class=~, properties=~}
 MetaSoundPresetBuilder {type=MetaSoundSourceBuilder, default=null}
 NewVar {type=S_MeleeSoundLibraryNameMatch, default {soundMatch="", names [], soundLibraryEvent {tagName=None}}}
 OldMantisData {type=FortMantisData, default=null}
 SelectedAssets {type="array<Object>", default []}
 SelectionTimerHandle {type=TimerHandle}
 SoundLibrary {type=SoundLibrary, default=null}
 SoundLibraryEvents {type="map<SoundLibraryTag>", default {}}
 SoundLibraryParent {type=SoundLibrary, default=/Game/Sounds/SoundLibrary/Parents/Weapon/Parent_WeaponMelee_Lib}
 SoundLibrarySounds {default {}}
 StringToStringMap {type="map<string>", default {}}
 Text {type=string, default=""}
 WeaponId {type="array<FortWeaponItemDefinition>", default []}
]
widgets {
 info {rootWidgetClass=&CanvasPanel, widgetCount=8}
 widgets [
  CanvasPanel_0 {widgetClassPath=&CanvasPanel, widget=CanvasPanel_0, properties {bOverride_Cursor=true}}
  EUW_MetaSoundAssetPath {
   widgetClassPath=/Game/Sounds/Fort_Audio_Blueprints/Utilities/EditorUtilityWidgets/Metasounds/Tools/EUW_MetaSoundAssetPath
   bIsVariable=true
   parent=CanvasPanel_0
   widget=EUW_MetaSoundAssetPath
   slot {
    class=CanvasPanelSlot
    properties {layoutData {alignment=<0 0>, anchors {maximum=<0 0>, minimum=<0 0>}, offsets=<ltrb 0 0 477.2864 191.6162>}}
   }
  }
  VerticalBox_0 {
   widgetClassPath=&VerticalBox
   parent=CanvasPanel_0
   widget=VerticalBox_0
   properties {bOverride_Cursor=true}
   slot {
    class=CanvasPanelSlot
    properties {layoutData {alignment=<0 0>, anchors {maximum=<0 0>, minimum=<0 0>}, offsets=<ltrb 0 196 479.1291 93.8138>}}
   }
  }
  DetailsView_135 {
   widgetClassPath=&DetailsView
   parent=VerticalBox_0
   widget=DetailsView_135
   properties {categoriesToShow [MeleeSettings]}
   slot {class=VerticalBoxSlot}
  }
  SoundLibButton {widgetClassPath=&EditorUtilityButton, parent=VerticalBox_0, widget=SoundLibButton, slot {class=VerticalBoxSlot}}
  TextBlock_81 {
   widgetClassPath=&TextBlock
   parent=SoundLibButton
   widget=TextBlock_81
   properties {text=Create Sound Library}
   slot {class=ButtonSlot, properties {padding=<ltrb 0 0 0 0>}}
  }
  EUW_Audio_Weapons_SoundName_ListView {
   widgetClassPath=/Game/Sounds/Fort_Audio_Blueprints/Utilities/EditorUtilityWidgets/Metasounds/Weapons/EUW_Audio_Weapons_SoundName_ListView
   bIsVariable=true
   parent=CanvasPanel_0
   widget=EUW_Audio_Weapons_SoundName_ListView
   properties {renderTransform {angle=0, scale=<1 1>, shear=<0 0>, translation=<0 0>}}
   slot {class=CanvasPanelSlot, properties {layoutData {alignment=<0 0>, anchors {maximum=<0 0>, minimum=<0 0>}, offsets=<ltrb 480 0 100 30>}}}
  }
  EUW_MeleeAudio_AnimationName_ListView {
   widgetClassPath=&/EUW_MeleeAudio_AnimationName_ListView
   bIsVariable=true
   parent=CanvasPanel_0
   widget=EUW_MeleeAudio_AnimationName_ListView
   slot {class=CanvasPanelSlot, properties {layoutData {alignment=<0 0>, anchors {maximum=<0 0>, minimum=<0 0>}, offsets=<ltrb 1084 0 100 30>}}}
  }
 ]
}
